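; Fizzy Cloud building actor: a timed smoke sphere that periodically applies a low-gravity gameplay effect, which also carries shield healing, to non-DBNO player pawns inside it.
; On initialization it sizes the area, can register as an AI occluder and starts its lifespan timer.
; SetFXMode's OnRep drives Niagara, audio and death effects, with smoke fades through custom primitive data.
; Overlapping clouds show only one sphere per pawn component, and water entry cues are cooldown-gated.
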(event EventDestroyed)

(event EventMulticast_HandleBinding
  (params (Actor :type Actor%20Object%20Reference) (HitComp :type Primitive%20Component%20Object%20Reference))
  ;[1 1 1] Handle Binding
  (|CallParentFunction Actor HitComp))

(event Custom|SetLocationFromServer
  (params (ServerLocation :type Vector))
  ;[1 1 1] Set Location From Server
  (Transformation|SetActorLocation self ServerLocation false true))

(event Building|EventOnBuildingActorInitialized
  (params (InitializationReason :type EFortBuildingInitializationReason%20Enum :default "StaticallyPlaced") (BuildingPersistentState :type EFortBuildingPersistentState%20Enum :default "Default"))
  (bind _returnvalue (Class|AthenaAIServiceZoneManager|GetAthenaAIServiceZoneManager true))
  (bind _row_lifespan (Variables|Default|GetRow_Lifespan))
  (bind _returnvalue_2 (ScalableFloat|GetValueatLevel _row_lifespan))
  ;[1 1 1] On Building Initialized
  (|CallParentFunction InitializationReason BuildingPersistentState)
  (Components|Sphere|SetSphereRadius (Variables|Default|GetAoECollision) (ScalableFloat|GetValueatLevel (Variables|Default|GetRow_AoeRadius)) true)
  (if (Variables|Default|GetRegisterAIOccluder)
    (Utilities|IsValid _returnvalue
      (:Is%20Valid
        (Class|AthenaAIServiceZoneManager|RegisterOccluder _returnvalue self (ScalableFloat|GetValueatLevel (Variables|Default|GetRow_AoeRadius))))
      (:Is%20Not%20Valid
        _))
    (else
      _))
  (AbilitySystem|SetTouchComponentforAbilityDeliveryInfo (Variables|Default|GetAoECollision))
  (Utilities|Time|SetTimerbyEvent (delegate Custom|KillSelf) _returnvalue_2)
  (Variables|Default|SetSetFXMode 1)
  (Variables|Default|SetEffectLength _returnvalue_2))

(event Custom|KillSelf
  ;[1 1 1] On Building Initialized
  (Variables|Default|SetDiedFromTimeOut true)
  (Gameplay|Die :self self :DamageCauser self)
  (Variables|Default|SetSetFXMode 0))

(event EventBeginPlay
  ;[1 1 1] Begin Play
  (|CallParentFunction)
  (Utilities|FlowControl|Sequence
    (:then_0
      (CallFunction|ResetEnterWaterEvent)
      (Audio|Components|Audio|Play (Variables|Default|GetEruptStartSound) 0.0)
      (CallFunction|SmokeFade true)
      (Networking|SwitchHasAuthority
        (:Authority
          (Health|BindEventtoOnBuildingHealthChanged (delegate Custom|HandleHealthChanged))
          (Utilities|Time|SetTimerbyEvent (EventDispatchers|CreateEvent :!event Custom|ApplyLowGrav) (ScalableFloat|GetValueatLevel (Variables|Default|GetRow%20ApplicationTickRate)) true)
          (CallFunction|ApplyLowGrav))))
    (:then_1
      (Variables|Default|SetOptimizeMultiOverlap (ScalableFloat|GetValueasBool (Variables|Default|GetSF_ToggleSphereVisibility)))
      (if (not (Variables|Default|GetOptimizeMultiOverlap))
        (Collision|SetCollisionResponsetoChannel (Variables|Default|GetAoECollision) "ECC_WorldDynamic" "ECR_Ignore")))
    (:then_2
      (Networking|SwitchHasAuthority
        (:Remote
          (if (not (Variables|Default|GetbResumeSimulation))
            (CallFunction|StopSim (Variables|Default|GetStopSimHitResult))))))
    (:then_3
      (if (ScalableFloat|GetValueasBool (Variables|Default|GetAVIndicatorEnabled))
        (SoundIndicator|StartSoundIndicatorLooping (Variables|Default|GetFortSoundIndicator) (Transformation|GetActorLocation) 0 1.0 0)))
    (:then_4
      (Networking|CanExecuteCosmeticEvents
        (:True
          (bind _bsuccessfullyactivated (Class|AutofireShroudComponent|ActivateShroudDetectors (Variables|Default|GetAutofireShroudComponent) (Variables|/Script/Engine.SphereComponent|GetSphereRadius (Variables|Default|GetAoECollision))))
          (if (not _bsuccessfullyactivated)
            (Fortnite|SetCanbeDamaged 0 true)))))))

(event Custom|HandleHealthChanged
  ;[1 1 1] Begin Play
  ;; Handle health changed
  (if (<= (Health|GetHealth) 0)
    (Variables|Default|SetSetFXMode 0)))

(event Custom|PlayFX
  ;[1 1 1] FX
  (bind _returnvalue (Niagara|SpawnSystemAttached "/FizzyCloudGameplay/Assets/FX/NS_FizzyCloud_Grenade_Form.NS_FizzyCloud_Grenade_Form" (Variables|Default|GetStaticMesh)))
  (Niagara|SetNiagaraVariableByString(Float) _returnvalue "NightMultiplier" (Variables|SmokeFX|GetSmoke_Night_Multiplier))
  (Components|Activation|Activate (Variables|Default|GetFX) true)
  ;;[0.8698 0 0 0.5] Scale Animation
  (bind (_direction _scale) (Timeline|Timeline Play)
    (:Update
      (Transformation|SetRelativeScale3D (Variables|Default|GetSM_FizzyCloud_Sphere) (Math|Conversions|ToVector(Float) (* _scale 1e+01))))))

(event Custom|EndFX
  ;[1 1 1] FX
  (Audio|Components|Audio|FadeOut (Variables|Default|GetEruptLoopSound) 2.5 0.0 "Linear" :!class /Script/Engine.AudioComponent)
  (Components|Activation|Deactivate (Variables|Default|GetFX))
  (Audio|Components|Audio|SetSound (Variables|Default|GetEruptStartSound) "/FizzyCloudGameplay/Assets/Audio/FizzyCloud_FizzyCloud_EruptStop_Cue.FizzyCloud_FizzyCloud_EruptStop_Cue")
  (SoundIndicator|StopSoundIndicatorLooping (Variables|Default|GetFortSoundIndicator))
  (Default|DeactivateShroudDetectors (Variables|Default|GetAutofireShroudComponent)))

(event Custom|Update%20FX%20for%20Water
  (params (Water%20Body :type Fort%20Water%20Body%20Actor%20Object%20Reference) (IsInWater :type Boolean))
  (bind _fx (Variables|Default|GetFX))
  ;[1 1 1] Updates Spray FX to account for water
  (bind _outwatersurfaceinfo (WaterBody|GetWaterSurfaceInfoStruct Water%20Body (Transformation|GetActorLocation) true false false))
  (Effects|Components|ParticleSystem|SetBoolParameter _fx "bisInWater" IsInWater)
  (Effects|Components|ParticleSystem|SetFloatParameter _fx "WaterBodyIndex" (Math|Conversions|ToFloat(Integer) (.WaterBodyIdx (Utilities|Struct|BreakWaterSurfaceInfo _outwatersurfaceinfo))))
  (Effects|Components|ParticleSystem|SetFloatParameter _fx "WaterDepth" (.WaterDepth (Utilities|Struct|BreakWaterSurfaceInfo _outwatersurfaceinfo)))
  (Effects|Components|ParticleSystem|SetFloatParameter _fx "WaveAttenuationFactor" (Variables|/Script/Water.WaterBodyComponent|GetTargetWaveMaskDepth (Water|GetWaterBodyComponent Water%20Body)))
  (Effects|Components|ParticleSystem|SetVectorParameter _fx "WaterPlaneLocation" (Utilities|Struct|BreakWaterSurfaceInfo _outwatersurfaceinfo))
  (Effects|Components|ParticleSystem|SetVectorParameter _fx "WaterPlaneNormal" (.WaterPlaneNormal (Utilities|Struct|BreakWaterSurfaceInfo _outwatersurfaceinfo))))

(event Custom|PlayDeathFX
  (bind _self self)
  (bind _returnvalue (Ability|EffectContext|GetInstigatorActor))
  (bind _returnvalue_location (Math|Transform|BreakTransform (Transformation|GetActorTransform _self)))
  (bind _returnvalue_2 (Ability|GameplayCue|MakeGameplayCueParameters :NormalizedMagnitude 0.0 :RawMagnitude 0.0 :Location _returnvalue_location :Normal "0,0,1.000000" :Instigator _returnvalue :EffectCauser _self :SourceObject _self :GameplayEffectLevel 1 :AbilityLevel 1 :bReplicateLocationWhenUsingMinimalRepProxy false))
  ;[1 1 1] Play death FX
  (CallFunction|SmokeFade false)
  (Niagara|SetNiagaraVariable(Bool) (Variables|Default|GetFX) "FadeOut" true)
  (Rendering|SetVisibility (Variables|Default|GetStaticMesh) false false)
  (Networking|SwitchHasAuthority
    (:Authority
      (Ability|SendGameplayEventtoActor self (Variables|Default|GetEvent_ShieldGeneratorDied))
      (if (Variables|Default|GetDiedFromTimeOut)
        (Ability|ExecuteGameplayCueWithParamsReplicated (Variables|Default|GetTag_GC_TimedOut) _returnvalue_2)
        (else
          (Ability|ExecuteGameplayCueWithParamsReplicated (Variables|Default|GetTag_GC_Destroyed) _returnvalue_2))))))

(event Custom|ApplyHealTagTracker)

(event EventMulticast_PlayHitFX
  (params (HitResult :type Hit%20Result%20Structure))
  ;[1 1 1] Bounce Land FX
  (Ability|ExecuteGameplayCueWithParamsReplicated (Variables|Default|GetTag_GC_HitWorld) (GameplayCue|MakeGameplayCueParametersfromHitResult HitResult)))

(event OnEnterWaterEvent(FortWaterInteraction)
  (params (WaterBody :type Fort%20Water%20Body%20Actor%20Object%20Reference) (WaterInteractionComponent :type Fort%20Water%20Interaction%20Component%20Object%20Reference) (bIsFirstBody :type Boolean))
  (bind _returnvalue_location (Math|Transform|BreakTransform (Transformation|GetActorTransform self)))
  ;[1 1 1] Water FX
  (CallFunction|UpdateFXForWater WaterBody true)
  (Utilities|FlowControl|DoOnce false
    (:Completed
      (Ability|ExecuteGameplayCueWithParamsReplicated (Variables|Default|GetTag_GC_HitWater) (Ability|GameplayCue|MakeGameplayCueParameters :NormalizedMagnitude 0.0 :RawMagnitude 0.0 :Location _returnvalue_location :Normal "0,0,1.000000" :Instigator (Ability|EffectContext|GetInstigatorActor) :EffectCauser self :SourceObject self :GameplayEffectLevel 1 :AbilityLevel 1 :bReplicateLocationWhenUsingMinimalRepProxy false))
      (Utilities|Time|SetTimerbyEvent (EventDispatchers|CreateEvent :!event Custom|ResetEnterWaterEvent) (ScalableFloat|GetValueatLevel (Variables|Default|GetRow_WaterSplashFXCooldown))))) ; Timed gate to prevent FX spam in super wavy waters
)

(event Custom|ResetEnterWaterEvent
  (bind _returnvalue_location (Math|Transform|BreakTransform (Transformation|GetActorTransform self)))
  ;[1 1 1] Water FX
  (Utilities|FlowControl|DoOnce false
    (:Completed
      (Ability|ExecuteGameplayCueWithParamsReplicated (Variables|Default|GetTag_GC_HitWater) (Ability|GameplayCue|MakeGameplayCueParameters :NormalizedMagnitude 0.0 :RawMagnitude 0.0 :Location _returnvalue_location :Normal "0,0,1.000000" :Instigator (Ability|EffectContext|GetInstigatorActor) :EffectCauser self :SourceObject self :GameplayEffectLevel 1 :AbilityLevel 1 :bReplicateLocationWhenUsingMinimalRepProxy false))
      (Utilities|Time|SetTimerbyEvent (EventDispatchers|CreateEvent :!event Custom|ResetEnterWaterEvent) (ScalableFloat|GetValueatLevel (Variables|Default|GetRow_WaterSplashFXCooldown))))) ; Timed gate to prevent FX spam in super wavy waters
)

(event Custom|ApplyLowGrav
  ; add low grav, shield healing is added through the low grav ge
  (for _array_element (Collision|GetOverlappingActors (Variables|Default|GetAoECollision) "/Script/Engine.Pawn" :!class /Script/Engine.PrimitiveComponent)
    (bind _asfort_player_pawn (Utilities|Casting|CastToFortPlayerPawn _array_element))
    (bind _asgameplay_tag_asset_interface (Utilities|Casting|CastToGameplayTagAssetInterface _asfort_player_pawn))
    (if (not (GameplayTags|HasMatchingGameplayTag _asgameplay_tag_asset_interface "(TagName=\"Gameplay.Status.DBNO\")"))
      (bind _asfort_ability_system_component (Utilities|Casting|CastToFortAbilitySystemComponent (Ability|GetAbilitySystemComponent (Variables|Actor|GetInstigator))))
      (GameplayEffects|ApplyGameplayEffectToTarget (Ability|GetAbilitySystemComponent 0) "/FizzyCloudGameplay/Gameplay/GE_FizzyCloud_GravBerry_LowGrav.GE_FizzyCloud_GravBerry_LowGrav_C" (Ability|GetAbilitySystemComponent _asfort_player_pawn) 0.0 (GameplayEffects|MakeFortEffectContext _asfort_ability_system_component 0 self 0 0)))))

(event OnExitWaterEvent(FortWaterInteraction)
  (params (WaterBody :type Fort%20Water%20Body%20Actor%20Object%20Reference) (WaterInteractionComponent :type Fort%20Water%20Interaction%20Component%20Object%20Reference) (bIsLastBody :type Boolean))
  ;[1 1 1] On Exit Water
  (CallFunction|UpdateFXForWater WaterBody))

(event EventEndPlay
  (params (EndPlayReason :type EEndPlayReason%20Enum :default "Destroyed"))
  (bind _returnvalue (Class|AthenaAIServiceZoneManager|GetAthenaAIServiceZoneManager true))
  ; On End Play
  (Utilities|IsValid _returnvalue
    (:Is%20Valid
      (Class|AthenaAIServiceZoneManager|UnregisterOccluder _returnvalue self))))

(fn GetDynamicDestructionResourceType
  (return "Metal"))

(fn GetDynamicDestructionStaticMeshComponents
  (return (Utilities|Array|MakeArray (Variables|Default|GetStaticMesh)) 0 0))

(fn HandleResumeSimulationChanged
  (|CallParentFunction)
  (Networking|SwitchHasAuthority
    (:Remote
      (if (Variables|Default|GetbResumeSimulation)
        (bind _returnvalue (Utilities|Time|SetTimerbyFunctionName 0 "RefreshMaterialPosition" 0.1 true true))
        (Variables|Default|SetRefreshPositionTimerHandler _returnvalue)
        (else
          (Utilities|Time|ClearandInvalidateTimerbyHandle (Variables|Default|GetRefreshPositionTimerHandler)))))))

(fn InitializeSmokeParameters
  (bind _fx (Variables|Default|GetFX))
  (bind _sm_fizzycloud_sphere (Variables|Default|GetSM_FizzyCloud_Sphere))
  (Utilities|FlowControl|Sequence
    (:then_1
      (Rendering|Material|SetDefaultCustomPrimitiveDataVector3 _sm_fizzycloud_sphere 3 (Math|Conversions|ToVector(LinearColor) (Variables|SmokeFX|GetSmokeColor)))
      (Rendering|Material|SetDefaultCustomPrimitiveDataVector3 _sm_fizzycloud_sphere 17 (Math|Conversions|ToVector(LinearColor) (Variables|SmokeFX|GetSmokeColor_Inter)))
      (Rendering|Material|SetDefaultCustomPrimitiveDataFloat _sm_fizzycloud_sphere 2 (Variables|SmokeFX|GetSmoke_Night_Multiplier))
      (Rendering|Material|SetDefaultCustomPrimitiveDataVector3 _sm_fizzycloud_sphere 12 (Math|Conversions|ToVector(LinearColor) (Variables|SmokeFX|GetSmokeNightColorMultiplier)))
      (Rendering|Material|SetDefaultCustomPrimitiveDataVector3 _sm_fizzycloud_sphere 8 (Math|Conversions|ToVector(LinearColor) (Variables|SmokeFX|GetSmokeEveningColorMultiplier)))
      (CallFunction|SmokeFade (and (Editor|InEditororPIE) (Variables|Default|GetSmokeDebug))))
    (:then_2
      (Niagara|SetNiagaraVariableByString(LinearColor) _fx "SmokeColor" (Variables|SmokeFX|GetParticle_SmokeColor))
      (Niagara|SetNiagaraVariableByString(Float) _fx "NightMultiplier" (Variables|SmokeFX|GetSmoke_Night_Multiplier)))))

(fn OnRep_SetFXMode
  (if (>= (Variables|Default|GetSetFXMode) 0)
    (if (>= (Variables|Default|GetSetFXMode) 1)
      (CallFunction|PlayFX)
      (else
        (CallFunction|EndFX)
        (CallFunction|PlayDeathFX)))))

(fn RefreshForMultiCloud
  (params (CloudPawnComp :type BP%20Fizzy%20Cloud%20Pawn%20Comp%20Object%20Reference))
  (LocalVariables|SetOverlappingCloudNumber (Utilities|Array|Length (Variables|/FizzyCloudGameplay/Gameplay/BP_FizzyCloud_PawnComp.BP_FizzyCloud_PawnComp_C|GetOverlappingFizzyCloud CloudPawnComp)))
  (if (> (LocalVariables|GetOverlappingCloudNumber) 0)
    (Rendering|SetVisibility (Variables|Default|GetSM_FizzyCloud_Sphere (Utilities|Array|Get(acopy) (Variables|/FizzyCloudGameplay/Gameplay/BP_FizzyCloud_PawnComp.BP_FizzyCloud_PawnComp_C|GetOverlappingFizzyCloud CloudPawnComp))) true true)
    (if (> (LocalVariables|GetOverlappingCloudNumber) 1)
      (for _index (range 1 (LocalVariables|GetOverlappingCloudNumber))
        (Rendering|SetVisibility (Variables|Default|GetSM_FizzyCloud_Sphere (Utilities|Array|Get(acopy) (Variables|/FizzyCloudGameplay/Gameplay/BP_FizzyCloud_PawnComp.BP_FizzyCloud_PawnComp_C|GetOverlappingFizzyCloud CloudPawnComp) _index)) false true)))))

(fn ShouldUseDynamicDestructionMaterial
  (return false))

(fn ShouldUseDynamicDestructionNiagaraSystem
  (return false))

(fn SmokeFade
  (params (FadeIn :type Boolean))
  (bind _sm_fizzycloud_sphere (Variables|Default|GetSM_FizzyCloud_Sphere))
  (Rendering|Material|SetDefaultCustomPrimitiveDataFloat _sm_fizzycloud_sphere 16 (Math|Float|SelectFloat 1.5 3.0 FadeIn))
  (Rendering|Material|SetDefaultCustomPrimitiveDataFloat _sm_fizzycloud_sphere 1 (Math|Conversions|ToFloat(Boolean) (not FadeIn)))
  (Rendering|Material|SetDefaultCustomPrimitiveDataFloat _sm_fizzycloud_sphere 0 (Utilities|Time|GetGameTimeinSeconds)))

(fn ConstructionScript
  (|CallParentFunction)
  (CallFunction|InitializeSmokeParameters))
