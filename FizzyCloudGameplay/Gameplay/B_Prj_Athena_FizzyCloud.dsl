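; Fizzy Cloud projectile Blueprint that triggers InitiateSpawnBGA on impact, presumably to spawn the cloud's gameplay actor.
; OnStop runs the parent logic, then passes on the hit as a catch-all for targets the parent misses, such as some vehicles.
; Water bodies are caught on begin overlap, since their hit is null in OnStop, and passed with a zero location.
; IsHitValidForSpawn flips its two boolean results when the hit actor has a moving platform component.
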(event EventBeginPlay)

(event Collision|EventActorBeginOverlap
  (params (OtherActor :type Actor%20Object%20Reference))
  ; Handling water collision as Hit comes up null during EventOnStop
  (bind _asfort_water_body_bp (Utilities|Casting|CastToFortWaterBodyBP OtherActor))
  (CallFunction|InitiateSpawnBGA :HitOtherActor _asfort_water_body_bp :HitLocation "0, 0, 0"))

(event EventOnStop
  (params (Hit :type Hit%20Result%20Structure%20%28by%20ref%29))
  (bind (_bblockinghit _binitialoverlap _time _distance _location _impactpoint _normal _impactnormal _physmat _hitactor _hitcomponent) (Collision|BreakHitResult Hit))
  (|CallParentFunction Hit)
  ; Sort of CatchAll of remaining actors/comps not caught by parent. I.e some vehicles
  (CallFunction|InitiateSpawnBGA Hit _hitactor _hitcomponent _location))

(fn IsHitValidForSpawn
  (params (HitOtherComp :type Primitive%20Component%20Object%20Reference) (HitOther :type Actor%20Object%20Reference) (HitResult :type Hit%20Result%20Structure))
  (Utilities|IsValid (Class|FortBaseMovingPlatformComponent|FindMovingPlatformComponentfromActor HitOther true)
    (:Is%20Valid
      (return false true))
    (:Is%20Not%20Valid
      (return true false))))

(fn ConstructionScript
  (|CallParentFunction))
